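# Niagara system for the SaltyBrim character intro: a blue speed-blur dash effect that flashes from alternating sides, ending in a burst.
# Sparks, DashTrail_Mesh and BorderFlash share four timed loops, using the loop count to alternate sides, with special motion timing on the last pass.
# Trails combine layered meshes and a sprite, sparks split into three spawn groups, and a local module pins flashes to camera borders.
# Final_FlashBurst fires a delayed central flash.
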NiagaraSystem /BRCosmetics/Animation/Game/MainPlayer/CharIntros/s42/SaltyBrim/FX/NS_CharIntros_SaltyBrim_BlurFX
dependencies {
 usedDataInterfaces [
  &NiagaraDataInterfaceCurve
  &NiagaraDataInterfaceVector2DCurve
  &NiagaraDataInterfaceVectorCurve
  &NiagaraDataInterfaceCamera
  &NiagaraDataInterfaceColorCurve
 ]
 usedDynamicInputs []
 usedModules [
  &SystemState
  &self:SystemSpawnScript.SystemScriptSource.SystemScriptGraph.NiagaraNodeAssignment_0.SetVariables_#1
 ]
}
emitters {
 BorderFlash {
  emitterUpdateScript [
   &EmitterState(
    Delay First Loop Only=true
    Life Cycle Mode=::NewEnumerator1
    Loop Behavior=::NewEnumerator2
    Loop Count=4
    Loop Delay=0.1
    Loop Duration=0.4375
    UseLoopDelay=true
   )
   &SpawnBurst_Instantaneous(Use Spawn Probability=false, UseLoopCountLimit=false)
  ]
  particleSpawnScript [
   /SetVariables(Particles.SpawnGroup=$Emitter.LoopCount) {isSetParameters=true}
   &InitializeParticle(
    AdjustAlpha=false
    AdjustHue=true
    AdjustSaturation=true
    AdjustValue=false
    Color=<rgba 0.2 1.89 8 1>
    Color Mode=::NewEnumerator3
    Hue Shift Range:Vector2f=<-0.01 0.01>
    Lifetime=0.2
    Position Offset=&MakeVector(
     X=0
     Y=&MergeHelper_FloatSwitch(0=100, 1=-100, 2=100, 3=-100, Index=$Particles.SpawnGroup, Num=::NewEnumerator2)
     Z=120
    )
    Sprite Rotation Angle=0.5
    Sprite Rotation Mode=::NewEnumerator1
    Sprite Size Mode=::NewEnumerator3
    Uniform Sprite Size=750
    UsePositionOffset=true
    Value Range:Vector2f=<1 1>
    Write Material Random=true
   )
   &CameraOffset(Camera Offset Amount=-100)
  ]
  particleUpdateScript [
   &ParticleState()
   &self:MoveParticleToCamBorder(Distance=350) {module=MoveParticleToCamBorder}
   /SetVariables(Particles.Position=&MergeHelper_VectorSwitch(
     0=Linked() {linkedVariable {name=Particles.MoveParticleToCamBorder.CamRightWS, type=NiagaraPosition}}
     1=Linked() {linkedVariable {name=Particles.MoveParticleToCamBorder.CamLeftWS, type=NiagaraPosition}}
     2=?
     3=?
     Index=&ModuloInt(A=$Emitter.LoopCount, B=2)
     Num=::NewEnumerator0
    )) {isSetParameters=true}
   &ScaleSpriteSize(Uniform Curve Sprite Scale=DataInterface() {propertyValues {Curve {
       defaultValue=1
       keys [
        {arriveTangent=9.0725, interpMode=RCIM_Cubic, leaveTangent=9.0725, tangentMode=RCTM_User, time=0, value=0}
        {arriveTangent=0.2675, interpMode=RCIM_Cubic, leaveTangent=0.2675, tangentMode=RCTM_User, time=0.1408, value=0.9847}
        {arriveTangent=-0.0405, interpMode=RCIM_Cubic, leaveTangent=-0.0405, tangentMode=RCTM_User, time=1, value=0}
       ]
      }}})
  ]
  rendererClasses [&NiagaraSpriteRendererProperties]
  renderers [{rendererClass=&NiagaraSpriteRendererProperties}]
 }
 DashTrail_Mesh {
  emitterUpdateScript [
   &EmitterState(
    Delay First Loop Only=true
    Life Cycle Mode=::NewEnumerator1
    Loop Behavior=::NewEnumerator2
    Loop Count=4
    Loop Delay=0.1
    Loop Duration=0.4375
    UseLoopDelay=true
   )
   &SpawnBurst_Instantaneous(Use Spawn Probability=false, UseLoopCountLimit=false)
  ]
  particleSpawnScript [
   /SetVariables(Particles.SpawnGroup=$Emitter.LoopCount) {isSetParameters=true}
   &InitializeParticle(
    Color Mode=::NewEnumerator0
    Lifetime=0.35
    Mesh Scale=<0.5 0.5 0.4>
    Mesh Scale Mode=::NewEnumerator4
    Position Offset=&MakeVector(
     X=0
     Y=&MergeHelper_FloatSwitch(0=300, 1=-300, 2=300, 3=-300, Index=$Particles.SpawnGroup, Num=::NewEnumerator2)
     Z=120
    )
    Sprite UV Mode=::NewEnumerator4
    Sprite UV Scale:Vector2f=<-1 1>
    UsePositionOffset=true
    Write Material Random=true
   )
   &InitialMeshOrientation(
    Mesh Orientation Mode=::NewEnumerator7
    Rotation=&MergeHelper_VectorSwitch(0=<0 0 0.5>, 1=<0 0 0>, 2=<0 0 0.5>, 3=<0 0 0>, Index=$Particles.SpawnGroup, Num=::NewEnumerator2)
    Use Rotation Vector=true
   )
   &CameraOffset(Camera Offset Amount=-40)
   &SpriteFacingAndAlignment(
    Alignment Coordinate Space=::World
    Facing Coordinate Space=::World
    Sprite Alignment=<0 0 1>
    Sprite Facing=&MakeVector(
     X=&MergeHelper_FloatSwitch(0=1, 1=-1, 2=?, 3=?, Index=&ModuloInt(A=$Particles.SpawnGroup, B=2), Num=::NewEnumerator0)
     Y=0
     Z=0
    )
    WriteAlignment=true
    WriteFacing=true
   )
  ]
  particleUpdateScript [
   &ParticleState()
   &ScaleMeshSize(Scale Factor=&MakeVector(
     X=&FloatFromCurve(FloatCurve=DataInterface() {propertyValues {Curve {keys [{arriveTangent=-4.0276, time=0.4, value=1}, <at 1 0>]}}})
     Y=1
     Z=1
    ))
   /SetVariables(Particles.Position=&AddVectorToPosition(
     Position=$Particles.Initial.Position
     Vector=&MakeVector(
      X=0
      Y=&MaskFloat(
       Mask Source=1
       Mask Parameter=$Particles.SpawnGroup
       Mask=3
       Masked Float=&FloatFromCurve(
        FloatCurve=DataInterface() {propertyValues {Curve {keys [
            {arriveTangent=4155.6484, interpMode=RCIM_Cubic, leaveTangent=4155.6484, tangentMode=RCTM_User, time=1.57, value=0}
            {arriveTangent=1602.4031, interpMode=RCIM_Cubic, leaveTangent=1602.4031, tangentMode=RCTM_User, time=1.67, value=275}
            <at 1.7333 320>
           ]}}}
        CurveIndex=$Engine.System.Age
       )
       Passthrough Float=&FloatFromCurve(
        FloatCurve=DataInterface() {propertyValues {Curve {keys [
            {arriveTangent=1754.0598, interpMode=RCIM_Cubic, leaveTangent=1754.0598, tangentMode=RCTM_User, time=0, value=1}
            {arriveTangent=153.3736, interpMode=RCIM_Cubic, leaveTangent=153.3736, tangentMode=RCTM_User, time=0.3, value=262.5}
            <at 1 300>
           ]}}}
        Scale Curve=&MergeHelper_FloatSwitch(0=-2, 1=2, 2=-2, 3=?, Index=$Particles.SpawnGroup, Num=::NewEnumerator1)
       )
      )
      Z=0
     )
    )) {isSetParameters=true}
   &ScaleColor(
    Scale Alpha=&FloatFromCurve(
     FloatCurve=DataInterface() {propertyValues {Curve {keys [
         {interpMode=RCIM_Cubic, leaveTangentWeight=0.3821, tangentMode=RCTM_User, time=0.8, value=1}
         {arriveTangent=-6.6613, tangentMode=RCTM_User, time=1, value=0}
        ]}}}
     Scale Curve=6
    )
    Scale RGB=&MakeVectorFromLinearColorRGB(Color=<rgba 0 0.0513 0.4564 1>)
    ScaleA=true
    ScaleRGB=true
   )
   &SolveForcesAndVelocity(Clamp Velocity=false, Limit Acceleration=false)
   &FN_Mesh_Attributes(
    Color=&MultiplyLinearColor(
     Linear Color=<rgba 0 0.0547 0.487 1>
     Scale Factor=&MakeVector4FromVectorAndFloat(
      "Vector (RGB)"=<1 1 1>
      "Float (Alpha)"=&FloatFromCurve(FloatCurve=DataInterface() {propertyValues {Curve {keys [<at 0.8 1>,<at 1 0>]}}})
     )
    )
    Create Color Attribute=true
    Create Scale Attribute=true
    Module Output=::NewEnumerator1
    Scale=&MakeVector(
     X=&FloatFromCurve(FloatCurve=DataInterface() {propertyValues {Curve {keys [<at 0.4 1>,<at 1 0>]}}})
     Y=1
     Z=0.6
    )
   )
   &FN_Mesh_Attributes(
    Color=&MultiplyLinearColor(
     Linear Color=<rgba 0 0.1667 1 0.1826>
     Scale Factor=&MakeVector4FromVectorAndFloat(
      "Vector (RGB)"=<1 1 1>
      "Float (Alpha)"=&FloatFromCurve(FloatCurve=DataInterface() {propertyValues {Curve {keys [<at 0.8 1>,<at 1 0>]}}}, Scale Curve=10)
     )
    )
    Create Color Attribute=true
    Create Scale Attribute=true
    Module Output=::NewEnumerator1
    Scale=&Multiply_Vector(A=$Particles.Scale, B=<0.7 1 0.95>)
   ) {module=FN_Mesh_Attributes001}
   &FN_Mesh_Attributes(
    Color=&MultiplyLinearColor(
     Linear Color=<rgba 0 0.3333 2 0.1826>
     Scale Factor=&MakeVector4FromVectorAndFloat(
      "Vector (RGB)"=<1 1 1>
      "Float (Alpha)"=&FloatFromCurve(FloatCurve=DataInterface() {propertyValues {Curve {keys [<at 0.8 1>,<at 1 0>]}}}, Scale Curve=10)
     )
    )
    Create Color Attribute=true
    Create Scale Attribute=true
    Module Output=::NewEnumerator1
    Scale=&Multiply_Vector(A=$Particles.Scale, B=<0.7 1 1>)
   ) {module=FN_Mesh_Attributes003}
   &FN_Mesh_Attributes(
    Color=&MultiplyLinearColor(
     Linear Color=<rgba 0.102 1.085 6 0.1826>
     Scale Factor=&MakeVector4FromVectorAndFloat(
      "Vector (RGB)"=<1 1 1>
      "Float (Alpha)"=&FloatFromCurve(FloatCurve=DataInterface() {propertyValues {Curve {keys [<at 0.8 1>,<at 1 0>]}}})
     )
    )
    Create Color Attribute=true
    Create Scale Attribute=true
    Module Output=::NewEnumerator1
    Scale=&Multiply_Vector(A=$Particles.Scale, B=<1.5 1 1>)
   ) {module=FN_Mesh_Attributes002}
   &FN_Sprite_Attributes(
    Color=&MultiplyLinearColor(
     Linear Color=<rgba 0.087 0.6601 5 10>
     Scale Factor=&MakeVector4FromVectorAndFloat(
      "Vector (RGB)"=<1 1 1>
      "Float (Alpha)"=&FloatFromCurve(FloatCurve=DataInterface() {propertyValues {Curve {keys [<at 0.8 1>,<at 1 0>]}}})
     )
    )
    Create Color Attribute=true
    Create Position Attribute=true
    Create SpriteSize Attribute=true
    Module Output:ENiagara_AttributeSamplingApplyOutput=::NewEnumerator1
    Position:NiagaraPosition=&AddVectorToPosition(Position=$Particles.Position, Vector=<0 0 -1.7504>)
    SpriteSize=&Multiply_Vector2D(
     A=<150 98>
     B=&MakeVector2D(
      X=&FloatFromCurve(FloatCurve=DataInterface() {propertyValues {Curve {keys [{arriveTangent=-4.0276, time=0.6, value=1}, <at 1 0>]}}})
      Y=1
     )
    )
   )
  ]
  rendererClasses [&NiagaraMeshRendererProperties, &NiagaraSpriteRendererProperties]
  renderers [
   {rendererClass=&NiagaraMeshRendererProperties, rendererIndex=0}
   {rendererClass=&NiagaraMeshRendererProperties, rendererIndex=1}
   {rendererClass=&NiagaraMeshRendererProperties, rendererIndex=2}
   {rendererClass=&NiagaraMeshRendererProperties, rendererIndex=3}
   {rendererClass=&NiagaraMeshRendererProperties, rendererIndex=4}
   {rendererClass=&NiagaraSpriteRendererProperties, rendererIndex=5}
  ]
 }
 Final_FlashBurst {
  emitterUpdateScript [
   &EmitterState(Life Cycle Mode=::NewEnumerator1, Loop Behavior=::NewEnumerator1, Loop Delay=1.6667, Loop Duration=1, UseLoopDelay=true)
   &SpawnBurst_Instantaneous(Use Spawn Probability=false, UseLoopCountLimit=false)
  ]
  particleSpawnScript [
   &InitializeParticle(
    AdjustAlpha=false
    AdjustHue=true
    AdjustSaturation=true
    AdjustValue=true
    Color=<rgba 0.0275 0.2599 1.1 1>
    Color Mode=::NewEnumerator3
    Hue Shift Range:Vector2f=<-0.01 0.01>
    Lifetime=0.2
    Position Offset=<0 -5 100>
    Sprite Rotation Angle=0.5
    Sprite Rotation Mode=::NewEnumerator1
    Sprite Size Mode=::NewEnumerator3
    Uniform Sprite Size=600
    UsePositionOffset=true
    Write Material Random=true
   )
   &CameraOffset(Camera Offset Amount=-50)
  ]
  particleUpdateScript [
   &ParticleState()
   &ScaleSpriteSize(Uniform Curve Sprite Scale=DataInterface() {propertyValues {Curve {
       defaultValue=1
       keys [
        {arriveTangent=9.0725, interpMode=RCIM_Cubic, leaveTangent=9.0725, tangentMode=RCTM_User, time=0, value=0}
        {arriveTangent=0.2675, interpMode=RCIM_Cubic, leaveTangent=0.2675, tangentMode=RCTM_User, time=0.1408, value=0.9847}
        {arriveTangent=-0.0405, interpMode=RCIM_Cubic, leaveTangent=-0.0405, tangentMode=RCTM_User, time=1, value=0}
       ]
      }}})
   &DynamicMaterialParameters(
    Index 0 Param 1=1
    Index 0 Param 2=1
    Index 0 Param 3=1
    Index 0 Param 4=1
    Module.Index0Float0Write=true
    Module.Index0Float1Write=true
    Module.Index0Float2Write=true
    Module.Index0Float3Write=true
   )
   &FN_Sprite_Attributes(
    Camera Offset=15
    Color=&MultiplyLinearColor(
     Linear Color=<rgba 0.0126 0.0701 0.3 1>
     Scale Factor=&MakeVector4FromVectorAndFloat(
      "Vector (RGB)"=<1 1 1>
      "Float (Alpha)"=&FloatFromCurve(
       FloatCurve=DataInterface() {propertyValues {Curve {keys [{arriveTangent=-3.0538, interpMode=RCIM_Cubic, time=0.2, value=1}, <at 1 0 RCIM_Cubic>]}}}
       Scale Curve=3
      )
     )
    )
    Create Camera Offset Attribute=true
    Create Color Attribute=true
    Create Dynamic Material Parameter Attributes=true
    Create Position Attribute=true
    Create SpriteSize Attribute=true
    Index 0 Param 1=2.5
    Index 0 Param 2=1
    Index 0 Param 3=0.7
    Module Output:ENiagara_AttributeSamplingApplyOutput=::NewEnumerator1
    Module.Index0Float0Write=true
    Module.Index0Float1Write=true
    Module.Index0Float2Write=true
    Module.Index0Float3Write=true
    Position:NiagaraPosition=&AddVectorToPosition(Position=$Particles.Initial.Position, Vector=<0 0 -10>)
    SpriteSize=&Multiply_Vector2D_ByFloat(Float=0.75, Vector2D=<165 225>)
   )
  ]
  rendererClasses [&NiagaraSpriteRendererProperties]
  renderers [
   {rendererClass=&NiagaraSpriteRendererProperties}
   {rendererClass=&NiagaraSpriteRendererProperties, rendererIndex=1}
  ]
 }
 Sparks {
  emitterUpdateScript [
   &EmitterState(
    Delay First Loop Only=true
    Life Cycle Mode=::NewEnumerator1
    Loop Behavior=::NewEnumerator2
    Loop Count=4
    Loop Delay=0.1
    Loop Duration=0.4375
    UseLoopDelay=true
   )
   &SpawnRate(
    SpawnRate=&FloatFromCurve(
     FloatCurve=DataInterface() {propertyValues {Curve {keys [<at 0 1 RCIM_Constant>,<at 0.0333 0 RCIM_Constant>]}}}
     CurveIndex=$Emitter.NormalizedLoopAge
     Scale Curve=&RandomRangeFloat(Minimum=444, Maximum=666)
    )
    Use Spawn Probability=false
   )
   &SpawnRate(
    SpawnGroup=1
    SpawnRate=&FloatFromCurve(
     FloatCurve=DataInterface() {propertyValues {Curve {keys [<at 0 1 RCIM_Constant>,<at 0.0333 0 RCIM_Constant>]}}}
     CurveIndex=$Emitter.NormalizedLoopAge
     Scale Curve=&RandomRangeFloat(Minimum=444, Maximum=555)
    )
    Use Spawn Probability=false
   ) {module=SpawnRate001}
   &SpawnRate(
    SpawnGroup=2
    SpawnRate=&FloatFromCurve(
     FloatCurve=DataInterface() {propertyValues {Curve {keys [<at 0 1 RCIM_Constant>,<at 0.0333 0 RCIM_Constant>]}}}
     CurveIndex=$Emitter.NormalizedLoopAge
     Scale Curve=&RandomRangeFloat(Minimum=333, Maximum=444)
    )
    Use Spawn Probability=false
   ) {module=SpawnRate002}
  ]
  particleSpawnScript [
   /SetVariables(Particles.VisibilityTag=$Emitter.SpawnGroup) {isSetParameters=true}
   /SetVariables(Particles.SpawnGroup=$Emitter.LoopCount) {isSetParameters=true}
   &InitializeParticle(
    Color=<rgba 0.375 4.4792 25 1>
    Lifetime=&MergeHelper_FloatSwitch(
     0=&MaskFloat(
      Mask Source=1
      Mask Parameter=$Particles.SpawnGroup
      Mask=3
      Masked Float=&UniformRangedFloat(Minimum=0.725, Maximum=1.15)
      Passthrough Float=&UniformRangedFloat(Minimum=0.4, Maximum=0.6)
     )
     1=&RandomRangeFloat(Minimum=0.4, Maximum=0.7)
     2=&RandomRangeFloat(Minimum=1, Maximum=1.25)
     3=?
     Index=$Particles.VisibilityTag
     Num=::NewEnumerator1
    )
    Mesh Scale Mode=::NewEnumerator1
    Mesh Uniform Scale Max=0.545
    Mesh Uniform Scale Min=0.35
    Position Offset=&MakeVector(
     X=0
     Y=&MergeHelper_FloatSwitch(0=300, 1=-300, 2=300, 3=-300, Index=$Particles.SpawnGroup, Num=::NewEnumerator2)
     Z=&MergeHelper_FloatSwitch(0=120, 1=100, 2=160, 3=?, Index=$Particles.VisibilityTag, Num=::NewEnumerator1)
    )
    Position Offset Coordinate Space=::World
    Sprite Size=&Multiply_Vector2(A=<2 8>, B=&Vector2DFromFloat(Value=&Multiply_Float(A=1, B=2)))
    Sprite Size Mode=::NewEnumerator4
    UsePositionOffset=true
    Write Material Random=true
   )
   &ShapeLocation(
    Box Size=&MergeHelper_VectorSwitch(0=<125 50 70>, 1=<25 50 20>, 2=<125 50 70>, 3=?, Index=$Particles.VisibilityTag, Num=::NewEnumerator1)
    MaskSpawnGroup=false
    Override Seed=false
    Rotation Coordinate Space=::World
    Shape Primitive=::NewEnumerator2
    Surface Only=false
    UseOwnerScale=true
   )
   &AddVelocity(
    Rotation Coordinate Space=::World
    Velocity=&RandomRangeVector(Minimum=<0 -500 0>, Maximum=<0 -800 0>)
    Velocity Speed Scale=&MergeHelper_FloatSwitch(
     0=&Multiply_Float(
      A=7
      B=&MergeHelper_FloatSwitch(0=1, 1=-1, 2=1, 3=-1, Index=$Particles.SpawnGroup, Num=::NewEnumerator2)
     )
     1=&Multiply_Float(
      A=&RandomRangeFloat(Minimum=1.95, Maximum=3)
      B=&MergeHelper_FloatSwitch(0=1, 1=-1, 2=1, 3=-2.5, Index=$Particles.SpawnGroup, Num=::NewEnumerator2)
     )
     2=&Multiply_Float(
      A=&RandomRangeFloat(Minimum=0.8, Maximum=1.4)
      B=&MergeHelper_FloatSwitch(0=1, 1=-1, 2=1, 3=-1, Index=$Particles.SpawnGroup, Num=::NewEnumerator2)
     )
     3=?
     Index=$Particles.VisibilityTag
     Num=::NewEnumerator1
    )
   )
   &AddRotationalVelocity(Rotation Rate=&MergeHelper_VectorSwitch(
     0=<0 0 0>
     1=<0 0 0>
     2=&RandomVector(Vector Scale=&RandomRangeFloat(Minimum=1, Maximum=2))
     3=?
     Index=$Particles.VisibilityTag
     Num=::NewEnumerator1
    ))
   &CameraOffset(Camera Offset Amount=&MergeHelper_FloatSwitch(
     0=0
     1=&RandomRangeFloat(Minimum=40, Maximum=50)
     2=?
     3=?
     Index=$Particles.VisibilityTag
     Num=::NewEnumerator0
    ))
   &InitialMeshOrientation(Use Rotation Vector=false)
  ]
  particleUpdateScript [
   &ParticleState()
   &ScaleMeshSize(Scale Factor=&VectorFromFloat(Value=&FloatFromCurve(FloatCurve=DataInterface() {propertyValues {Curve {keys [<at 0.6 1 RCIM_Cubic>, {arriveTangent=-6.308, interpMode=RCIM_Cubic, leaveTangent=-6.308, tangentMode=RCTM_User, time=1, value=0}]}}})))
   &ScaleColor(
    Scale Alpha=&FloatFromCurve(FloatCurve=DataInterface() {propertyValues {Curve {keys [
         {arriveTangent=11.9433, interpMode=RCIM_Cubic, leaveTangent=11.9433, tangentMode=RCTM_User, time=0, value=0}
         <at 0.2 1 RCIM_Cubic>
         <at 1 0 RCIM_Cubic>
        ]}}})
    ScaleA=true
    ScaleRGB=true
   )
   &ScaleSpriteSize(
    Scale Factor=&Vector2DFromCurve(
     Vector2Curve=DataInterface() {propertyValues {
       XCurve {keys [<at 0 1>,<at 1 0>]}
       YCurve {keys [
         {arriveTangent=-11.1847, interpMode=RCIM_Cubic, leaveTangent=-11.1847, tangentMode=RCTM_User, time=0, value=4}
         {arriveTangent=-3.7, interpMode=RCIM_Cubic, leaveTangent=-3.7, time=0.3897, value=0.99}
         <at 1 0.3>
        ]}
      }}
     ScaleCurve=<1 1>
    )
    Scale Sprite Size Mode=::NewEnumerator2
   )
   &ScaleVelocity(
    Coordinate Space:ENiagaraCoordinateSpace=::World
    Velocity Scale=&MergeHelper_VectorSwitch(
     0=&Multiply_VectorByFloat(
      Vector=&MergeHelper_VectorSwitch(
       0=<1 1 1>
       1=<1 1 1>
       2=<1 1 1>
       3=&VectorFromFloat(Value=&FloatFromCurve(FloatCurve=DataInterface() {propertyValues {Curve {keys [<at 1.5 0 RCIM_Constant>,<at 1.55 1 RCIM_Constant>]}}}, CurveIndex=$Emitter.Age))
       Index=$Particles.SpawnGroup
       Num=::NewEnumerator2
      )
      Float=&MaskFloat(
       Mask Source=1
       Mask Parameter=$Particles.SpawnGroup
       Mask=3
       Masked Float=&FloatFromCurve(FloatCurve=DataInterface() {propertyValues {Curve {keys [
            {arriveTangent=-2.2293, interpMode=RCIM_Cubic, leaveTangent=-2.2293, tangentMode=RCTM_User, time=0, value=1}
            {arriveTangent=-0.6108, interpMode=RCIM_Cubic, leaveTangent=-0.6108, tangentMode=RCTM_User, time=0.3, value=0.15}
            <at 1 0.015 RCIM_Cubic>
           ]}}})
       Passthrough Float=&FloatFromCurve(FloatCurve=DataInterface() {propertyValues {Curve {keys [{arriveTangent=-2.2293, interpMode=RCIM_Cubic, leaveTangent=-2.2293, tangentMode=RCTM_User, time=0, value=1}, <at 1 0.1 RCIM_Cubic>]}}})
      )
     )
     1=&Multiply_VectorByFloat(
      Vector=&MergeHelper_VectorSwitch(
       0=<1 1 1>
       1=<1 1 1>
       2=<1 1 1>
       3=&VectorFromFloat(Value=&FloatFromCurve(FloatCurve=DataInterface() {propertyValues {Curve {keys [<at 1.5 0 RCIM_Constant>,<at 1.55 1 RCIM_Constant>]}}}, CurveIndex=$Emitter.Age))
       Index=$Particles.SpawnGroup
       Num=::NewEnumerator2
      )
      Float=&MaskFloat(
       Mask Source=1
       Mask Parameter=$Particles.SpawnGroup
       Mask=3
       Masked Float=&FloatFromCurve(FloatCurve=DataInterface() {propertyValues {Curve {keys [
            {arriveTangent=-2.2293, interpMode=RCIM_Cubic, leaveTangent=-2.2293, tangentMode=RCTM_User, time=0, value=1}
            {arriveTangent=-0.9034, interpMode=RCIM_Cubic, leaveTangent=-0.9034, tangentMode=RCTM_User, time=0.3, value=0.32}
            <at 1 0.04 RCIM_Cubic>
           ]}}})
       Passthrough Float=&FloatFromCurve(FloatCurve=DataInterface() {propertyValues {Curve {keys [{arriveTangent=-2.2293, interpMode=RCIM_Cubic, leaveTangent=-2.2293, tangentMode=RCTM_User, time=0, value=1}, <at 1 0.1 RCIM_Cubic>]}}})
      )
     )
     2=&Multiply_Vector(
      A=&MergeHelper_VectorSwitch(
       0=<1 1 1>
       1=<1 1 1>
       2=<1 1 1>
       3=&VectorFromFloat(Value=&FloatFromCurve(FloatCurve=DataInterface() {propertyValues {Curve {keys [<at 1.5 0 RCIM_Constant>,<at 1.55 1 RCIM_Constant>]}}}, CurveIndex=$Emitter.Age))
       Index=$Particles.SpawnGroup
       Num=::NewEnumerator2
      )
      B=&VectorFromCurve(
       VectorCurve=DataInterface() {propertyValues {
         XCurve {keys [{arriveTangent=-2.1556, interpMode=RCIM_Cubic, leaveTangent=-2.1556, tangentMode=RCTM_User, time=0, value=1}, <at 1 0>]}
         YCurve {keys [{arriveTangent=-2.1556, interpMode=RCIM_Cubic, leaveTangent=-2.1556, tangentMode=RCTM_User, time=0, value=1}, <at 1 0>]}
         ZCurve {keys [
           <at 0 1 RCIM_Cubic>
           {arriveTangent=-2.3309, interpMode=RCIM_Cubic, leaveTangent=-2.3309, tangentMode=RCTM_User, time=0.2, value=2}
           <at 1 1>
          ]}
        }}
       Scale Curve=<2 2 1>
      )
     )
     3=?
     Index=$Particles.VisibilityTag
     Num=::NewEnumerator1
    )
   )
   &AccelerationForce(Acceleration=&MergeHelper_VectorSwitch(
     0=<0 0 0>
     1=&RandomRangeVector(Minimum=<0 0 0>, Maximum=<0 0 0>)
     2=<0 0 -10000>
     3=?
     Index=$Particles.VisibilityTag
     Num=::NewEnumerator1
    ))
   &SolveForcesAndVelocity(Clamp Velocity=false, Limit Acceleration=false)
   &ScaleSpriteSizeBySpeed(Max Scale Factor=<1 4>, Min Scale Factor=<1 2>)
   &DynamicMaterialParameters(
    Index 0 Param 1=14.2084
    Index 0 Param 2=3.14
    Module.Index0Float0Write=true
    Module.Index0Float1Write=true
    Module.Index0Float2Write=true
    Module.Index0Float3Write=true
   )
   &DynamicMaterialParameters(
    Dynamic Parameter Output:ENiagara_AttributeSamplingApplyOutput=::NewEnumerator1
    Index 0 Param 1=&Add_Float(
     A=&RandomRangeFloat(Minimum=0, Maximum=1)
     B=&FloatFromCurve(
      FloatCurve=DataInterface() {propertyValues {Curve {keys [
          <at 0 0 RCIM_Cubic>
          {arriveTangent=1.5947, interpMode=RCIM_Cubic, leaveTangent=1.5947, tangentMode=RCTM_User, time=0.2557, value=0.5257}
          <at 1 1 RCIM_Cubic>
         ]}}}
      CurveIndex=$Particles.Age
      Scale Curve=0.25
     )
    )
    Module.Index0Float0Write=true
    Module.Index0Float1Write=true
    Module.Index0Float2Write=true
    Module.Index0Float3Write=true
   ) {module=DynamicMaterialParameters001}
   &DynamicMaterialParameters(
    Dynamic Parameter Output:ENiagara_AttributeSamplingApplyOutput=::NewEnumerator1
    Module.Index0Float0Write=true
    Module.Index0Float1Write=true
    Module.Index0Float2Write=false
    Module.Index0Float3Write=false
   ) {module=DynamicMaterialParameters002}
   /SetVariables(
    Particles.GlowColor=&MakeLinearColorFromVectorAndFloat(
     "Vector (RGB)"=&MakeVectorFromLinearColorRGB(Color=<rgba 0.1696 0.6195 1 1>)
     "Float (Alpha)"=&FloatFromCurve(FloatCurve=DataInterface() {propertyValues {Curve {keys [<at 0.5 1>,<at 1 0>]}}}, Scale Curve=0.5)
    )
    Particles.GlowSize=&Multiply_Vector2D(A=$Particles.SpriteSize, B=<8 2>)
   ) {isSetParameters=true}
   &FN_Sprite_Attributes(
    Color=&MultiplyLinearColor(
     Linear Color=&RandomRangeLinearColor(Minimum=<rgba 0.8 0.917 1 0.066>, Maximum=<rgba 0.76 1 1 0.091>)
     Scale Factor=&MakeVector4FromVectorAndFloat(
      "Vector (RGB)"=<1 1 1>
      "Float (Alpha)"=&FloatFromCurve(FloatCurve=DataInterface() {propertyValues {Curve {keys [
           {arriveTangent=2.4358, interpMode=RCIM_Cubic, leaveTangent=2.4358, tangentMode=RCTM_User, time=0, value=0}
           <at 0.25 0.5 RCIM_Cubic>
           <at 1 0 RCIM_Cubic>
          ]}}})
     )
    )
    Create Color Attribute=true
    Create SpriteSize Attribute=true
    Module Output:ENiagara_AttributeSamplingApplyOutput=::NewEnumerator1
    SpriteSize=&Multiply_Vector2D_ByFloat(
     Float=&FloatFromCurve(FloatCurve=DataInterface() {propertyValues {Curve {keys [
          {arriveTangent=1.7761, interpMode=RCIM_Cubic, leaveTangent=1.7761, tangentMode=RCTM_User, time=0, value=0}
          {interpMode=RCIM_Cubic, tangentMode=RCTM_User, time=1, value=1}
         ]}}})
     Vector2D=&Vector2DFromFloat(Value=&RandomRangeFloat(Minimum=120, Maximum=130))
    )
   )
   &FN_Mesh_Attributes(Create Color Attribute=true, Module Output=::NewEnumerator1)
  ]
  rendererClasses [&NiagaraSpriteRendererProperties, &NiagaraMeshRendererProperties]
  renderers [
   {rendererClass=&NiagaraSpriteRendererProperties}
   {rendererClass=&NiagaraSpriteRendererProperties, rendererIndex=1}
   {rendererClass=&NiagaraSpriteRendererProperties, rendererIndex=2}
   {rendererClass=&NiagaraSpriteRendererProperties, rendererIndex=3}
   {rendererClass=&NiagaraMeshRendererProperties, rendererIndex=4}
  ]
 }
}
summary {
 systemName=NS_CharIntros_SaltyBrim_BlurFX
 emitters [
  {emitterName=Sparks, rendererClasses [&NiagaraSpriteRendererProperties, &NiagaraMeshRendererProperties]}
  {emitterName=BorderFlash, rendererClasses [&NiagaraSpriteRendererProperties]}
  {emitterName=DashTrail_Mesh, rendererClasses [&NiagaraMeshRendererProperties, &NiagaraSpriteRendererProperties]}
  {emitterName=Final_FlashBurst, rendererClasses [&NiagaraSpriteRendererProperties]}
 ]
 userVariables []
}
